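# Parent material for the Soda Mug backpack cosmetic, adding an animated, color-cycling LED glow to a wheel on the mesh.
# An FX texture and UV1 form the wheel mask, while polar coordinates on UV1 supply depth falloff and time-scrolled animation.
# A looping three-color blend and cosine pulses build the emissive, which is added between the shared 2023 character PreFX and PostFX functions.
# Several alternative glow branches are left unconnected.
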Material /BRCosmetics/Accessories/FORT_Backpacks/Backpack_F_MED_SodaMug/Materials/M_SodaMug_Backpack_PARENT
graph {
 properties {
  bAutomaticallySetUsageInEditor=false
  bHasPixelAnimation=true
  bUseMaterialAttributes=true
  bUsedWithMeshDeformer=true
  bUsedWithSkeletalMesh=true
  pixelDepthOffsetMode=PDOM_Legacy
  previewMesh=/BRCosmetics/Accessories/FORT_Backpacks/Backpack_F_MED_SodaMug/Meshes/F_MED_SodaMug_Pack
  shadingModel=MSM_FromMaterialExpression
  usedShadingModels=Default Lit | Subsurface | Clear Coat | Subsurface Profile | Cloth
 }
 nodes [
  n1 = Multiply(A=NamedRerouteUsage()) {constB=3}
  n2 = Add(A=Multiply(A=Cosine(None=@n1)) {constB=0.5}) {constB=0.5}
  n3 = Multiply(A=@n2) {constB=0.3}
  Add(A=@n3) {constB=0.2}
  n4 = Add(A=NamedRerouteUsage(), B=Multiply(A=Divide() {constB=12}, B=Constant() {r=6})) {constB=0}
  n5 = Time()
  n6 = Round(None=Frac(None=Multiply(A=@n5) {constB=1.333}))
  n7 = Cosine(None=LinearInterpolate(A=@n4, B=Add(A=@n4) {constB=0.5}, Alpha=@n6))
  n8 = Add(A=Multiply(A=@n7) {constB=0.5}) {constB=0.5}
  n9 = Multiply(A=Multiply(A=@n8, B=@n8)) {constB=0.2}
  n10 = Saturate(None=Power(Base=@n8) {constExponent=64})
  Add(A=Multiply(A=Add(A=@n9, B=@n10), B=Saturate(None=@n7))) {constB=0.05, desc=glow anim}
  n11 = NamedRerouteUsage()
  n12 = Add(A=Multiply(A=Cosine(None=@n11)) {constB=0.5}) {constB=0.5}
  n13 = Multiply(A=@n12, B=@n12)
  n14 = Power(Base=@n12) {constExponent=64}
  n15 = Add(A=@n11) {constB=0.5}
  n16 = Add(A=Multiply(A=Cosine(None=@n15)) {constB=0.5}) {constB=0.5}
  n17 = Multiply(A=@n16, B=@n16)
  n18 = Power(Base=@n16) {constExponent=64}
  Add(A=Add(A=Multiply(A=@n13) {constB=0.2}, B=Saturate(None=@n14)), B=Add(A=Multiply(A=@n17) {constB=0.2}, B=Saturate(None=@n18)))
  Add(A=@n1) {constB=-0.1665}
  n19 = Add(A=@n3, B=Saturate(None=Power(Base=@n2) {constExponent=32}))
  LinearInterpolate(B=@n19, Alpha=@n6)
  n20 = NamedRerouteUsage()
  n21 = Power(Base=MaterialFunctionCall!(
    A=Constant3Vector() {constant=<rgba 1 0 0.9613 1>}
    B=Constant3Vector() {constant=<rgba 1 1 0 1>}
    C=Constant3Vector() {constant=<rgba 0 0.3925 1 1>}
    Alpha=Frac(None=@n20)
   ) {materialFunction=/Game/Global/MaterialFunctions/TechArt/MF_3ColorBlendLoop}.Result) {constExponent=6}
  LinearInterpolate(B=@n21, Alpha=@n6) {constA=0.5}
  n22 = MaterialFunctionCall() {materialFunction=&MF_CharacterParent_2023_PreFX}
  n23 = GetMaterialAttributes(None=@n22.Result) {attributeGetTypes [#1, #2]}
  n24 = NamedRerouteUsage()
  n25 = Cosine(None=Multiply(A=@n20) {constB=15})
  n26 = Saturate(None=Add(A=Multiply(A=@n25) {constB=0.1})) {desc=LEDs}
  n27 = Add(A=@n19) {constB=0.05, desc=glow anim}
  n28 = Multiply(A=Multiply(A=@n21, B=Multiply(A=@n26, B=@n27)), B=ScalarParameter() {defaultValue=20, parameterName=Emissive_Intensity, sliderMax=50}) {constB=20}
  n29 = Add(A=@n23.Emissive Color, B=Multiply(A=NamedRerouteUsage(), B=Multiply(A=@n24, B=@n28)))
  MaterialFunctionCall(
   Material=BlendMaterialAttributes(
    A=@n22.Result
    B=SetMaterialAttributes!(MaterialAttributes=@n23.MaterialAttributes, Emissive Color=@n29, Base Color=@n23.MaterialAttributes) {attributeSetTypes [#1, #2]}
    Alpha=NamedRerouteUsage()
   )
   M=@n22.M
   Pre Skinned Local Position=@n22.M
   Distance Blend=@n22.M
   Fresnel=@n22.M
  ) {materialFunction=&MF_CharacterParent_2023_PostFX}
  n30 = TextureSample() {samplerType=SAMPLERTYPE_Masks, texture=/BRCosmetics/Accessories/FORT_Backpacks/Backpack_F_MED_SodaMug/Textures/T_SodaMug_Backpack_FX}
  n31 = TextureCoordinate() {coordinateIndex=1}
  n32 = Ceil(None=ComponentMask(None=@n31) {r=true})
  NamedRerouteDeclaration(None=Multiply(A=@n30.R, B=@n32)) {name=WheelMask, nodeColor=<rgba 0 0.6719 1 1>}
  n33 = MaterialFunctionCall(Float2=MaterialFunctionCall!(UV Coord=@n31, LenghtScale=Constant() {r=2}, RadialScale=Constant() {r=1}) {materialFunction=/Game/Global/MaterialFunctions/TechArt/MF_PolarCoordinates}.Result) {materialFunction=&BreakOutFloat2Components}
  n34 = Saturate(None=Add(A=Power(Base=@n33.R) {constExponent=5}) {constB=0.2})
  NamedRerouteDeclaration(None=@n34) {name=WheelMaskDepth, nodeColor=<rgba 0 0.8359 1 1>}
  NamedRerouteDeclaration(None=@n32) {name=UVMask, nodeColor=<rgba 0 0.6719 1 1>}
  n35 = Time()
  NamedRerouteDeclaration(None=Add(A=Divide(A=@n33.G), B=Multiply(A=@n35) {constB=0.25})) {name=WheelLenghAnim, nodeColor=<rgba 0.7578 0 1 1>}
  OneMinus(None=@n34)
  Power(Base=@n2) {constExponent=1}
  n36 = OneMinus(None=@n33.R)
  Saturate(None=Add(A=Power(Base=@n36)) {constB=0.2})
 ]
}
