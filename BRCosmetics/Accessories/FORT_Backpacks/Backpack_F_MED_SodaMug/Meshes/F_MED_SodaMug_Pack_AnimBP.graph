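# Animation Blueprint for the Soda Mug backpack cosmetic that orients the pack against the wearer and adds secondary motion.
# On customization load it caches the pawn's part mesh, or the custom base mesh if the cast fails.
# Each update stores that mesh's pelvis location, still debug-printed, as the pack bone's Look At target.
# An additive offset follows, then six chained Anim Dynamics nodes swing the left and right roller bones with per-bone gravity.
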AnimBlueprint /BRCosmetics/Accessories/FORT_Backpacks/Backpack_F_MED_SodaMug/Meshes/F_MED_SodaMug_Pack_AnimBP
target_skeleton=&/F_MED_SodaMug_Pack_Skeleton
graphs [
 AnimGraph {nodes [
   n1 = Look At - Bone: pack!(ComponentPose=Local To Component(LocalPose=Input Pose()), Alpha=1, LookAtLocation:Vector=Get PelvisLocation() -> (PelvisLocation:Vector)) -> (Pose) {
    alphaBoolBlend {blendOption=Linear}
    boneToModify {boneName=pack}
    lookAt_Axis {axis {y=0, z=-1}}
    lookUp_Axis {axis {x=0, z=1}}
    lookAtTarget=~
   }
   n2 = "Transform (Modify) Bone - Bone: pack"(ComponentPose=@n1, Translation="0,3,-4", Rotation="0,0,63") {
    rotationMode=BMM_Additive
    rotationSpace=BCS_BoneSpace
    translationMode=BMM_Additive
    translationSpace=BCS_ParentBoneSpace
    alphaBoolBlend {blendOption=Linear}
    boneToModify {boneName=pack}
    rotation {roll=63}
    translation {y=3, z=-4}
   }
   n3 = Anim Dynamics - Bone: dyn_roller_r_01(ComponentPose=@n2, Alpha=1, ExternalForce="0,0,0") {
    angularDampingOverride=0.97
    angularSpringConstant=8
    bOverrideAngularDamping=true
    bUseGravityOverride=true
    linearSpringConstant=2
    alphaBoolBlend {blendOption=Linear}
    boundBone {boneName=dyn_roller_r_01}
    componentAppliedLinearAccClamp=<300 300 300>
    componentLinearAccScale=<0.3 0.3 0.3>
    componentLinearVelScale=<0.3 0.3 0.3>
    gravityOverride {z=-150}
    physicsBodyDefinitions [{
      boundBone {boneName=dyn_roller_r_01}
      boxExtents=<8 8 4>
      constraintSetup {
       angularLimitsMax=<3 90 3>
       angularLimitsMin=<-3 -90 -3>
       angularTarget=<0 0 0>
       linearAxesMax=<0.05 0.05 0.05>
       linearAxesMin=<-0.05 -0.05 -0.05>
      }
      localJointOffset=<0 0 -2>
     }]
   }
   n4 = Anim Dynamics - Bone: dyn_roller_r_02(ComponentPose=@n3, Alpha=1, ExternalForce="0,0,0") {
    angularDampingOverride=0.97
    angularSpringConstant=8
    bOverrideAngularDamping=true
    bUseGravityOverride=true
    linearSpringConstant=2
    alphaBoolBlend {blendOption=Linear}
    boundBone {boneName=dyn_roller_r_02}
    componentAppliedLinearAccClamp=<300 300 300>
    componentLinearAccScale=<0.3 0.3 0.3>
    componentLinearVelScale=<0.3 0.3 0.3>
    gravityOverride {z=-100}
    physicsBodyDefinitions [{
      boundBone {boneName=dyn_roller_r_02}
      boxExtents=<10 5 5>
      constraintSetup {angularLimitsMax=<5 90 5>, angularLimitsMin=<-5 -90 -5>, angularTarget=<0 0 0>}
      localJointOffset=<0 0 -2>
     }]
   }
   n5 = Anim Dynamics - Bone: dyn_roller_r_03(ComponentPose=@n4, Alpha=1, ExternalForce="0,0,0") {
    angularDampingOverride=0.97
    angularSpringConstant=8
    bOverrideAngularDamping=true
    bUseGravityOverride=true
    linearSpringConstant=2
    alphaBoolBlend {blendOption=Linear}
    boundBone {boneName=dyn_roller_r_03}
    componentAppliedLinearAccClamp=<300 300 300>
    componentLinearAccScale=<0.3 0.3 0.3>
    componentLinearVelScale=<0.3 0.3 0.3>
    gravityOverride {z=-300}
    physicsBodyDefinitions [{
      boundBone {boneName=dyn_roller_r_03}
      boxExtents=<10 10 5>
      constraintSetup {angularLimitsMax=<5 90 5>, angularLimitsMin=<-5 -90 -5>, angularTarget=<0 0 0>}
      localJointOffset=<0 0 -2>
     }]
   }
   n6 = Anim Dynamics - Bone: dyn_roller_l_01(ComponentPose=@n5, Alpha=1, ExternalForce="0,0,0") {
    angularDampingOverride=0.97
    angularSpringConstant=8
    bOverrideAngularDamping=true
    bUseGravityOverride=true
    linearSpringConstant=2
    alphaBoolBlend {blendOption=Linear}
    boundBone {boneName=dyn_roller_l_01}
    componentAppliedLinearAccClamp=<300 300 300>
    componentLinearAccScale=<0.3 0.3 0.3>
    componentLinearVelScale=<0.3 0.3 0.3>
    gravityOverride {z=-100}
    physicsBodyDefinitions [{
      boundBone {boneName=dyn_roller_l_01}
      boxExtents=<8 8 5>
      constraintSetup {
       angularLimitsMax=<3 90 3>
       angularLimitsMin=<-3 -90 -3>
       angularTarget=<0 0 0>
       linearAxesMax=<0.05 0.05 0.05>
       linearAxesMin=<-0.05 -0.05 -0.05>
      }
      localJointOffset=<0 0 -2>
     }]
   }
   n7 = Anim Dynamics - Bone: dyn_roller_l_02(ComponentPose=@n6, Alpha=1, ExternalForce="0,0,0") {
    angularDampingOverride=0.97
    angularSpringConstant=8
    bOverrideAngularDamping=true
    bUseGravityOverride=true
    linearSpringConstant=2
    alphaBoolBlend {blendOption=Linear}
    boundBone {boneName=dyn_roller_l_02}
    componentAppliedLinearAccClamp=<300 300 300>
    componentLinearAccScale=<0.3 0.3 0.3>
    componentLinearVelScale=<0.3 0.3 0.3>
    gravityOverride {z=-200}
    physicsBodyDefinitions [{
      boundBone {boneName=dyn_roller_l_02}
      boxExtents=<10 10 5>
      constraintSetup {angularLimitsMax=<5 90 5>, angularLimitsMin=<-5 -90 -5>, angularTarget=<0 0 0>}
      localJointOffset=<0 0 -2>
     }]
   }
   n8 = Anim Dynamics - Bone: dyn_roller_l_03(ComponentPose=@n7, Alpha=1, ExternalForce="0,0,0") {
    angularDampingOverride=0.97
    angularSpringConstant=8
    bOverrideAngularDamping=true
    bUseGravityOverride=true
    linearSpringConstant=2
    alphaBoolBlend {blendOption=Linear}
    boundBone {boneName=dyn_roller_l_03}
    componentAppliedLinearAccClamp=<300 300 300>
    componentLinearAccScale=<0.3 0.3 0.3>
    componentLinearVelScale=<0.3 0.3 0.3>
    gravityOverride {z=-200}
    physicsBodyDefinitions [{
      boundBone {boneName=dyn_roller_l_03}
      boxExtents=<10 10 6>
      constraintSetup {angularLimitsMax=<5 90 5>, angularLimitsMin=<-5 -90 -5>, angularTarget=<0 0 0>}
      localJointOffset=<0 0 -2>
     }]
   }
   Output Pose(Result=Component To Local(ComponentPose=@n8)) {layerGroup=None}
  ]}
 EventGraph {nodes [
   Event BlueprintUpdateAnimation()
   n1 = Get bodyMesh() -> (bodyMesh:Skeletal Mesh Component Object Reference)
   > Is Valid(InputObject=@n1)
   > n2 = Set PelvisLocation(PelvisLocation:Vector=GetSocketLocation!(self:Scene Component Object Reference=@n1, InSocketName:Name=pelvis) -> (ReturnValue:Vector)) -> (Output_Get:Vector)
   > PrintString!(execute, InString="To String (Vector)"!(InVec:Vector=@n2) -> (ReturnValue:String), bPrintToScreen, bPrintToLog, TextColor, Duration=0, Key) -> (then)
   n3 = Event OnCharacterCustomizationLoaded()
   > n4 = Cast To FortPlayerPawnAthena(Object=TryGetPawnOwner())
   > Set bodyMesh(bodyMesh:Skeletal Mesh Component Object Reference=GetSkeletalMeshForPartType(self=@n4)) -> (Output_Get:Skeletal Mesh Component Object Reference)
   Set bodyMesh(execute=@n4.CastFailed, bodyMesh=@n3.CustomBaseMesh)
  ]}
]
variables [
 bodyMesh {type=SkeletalMeshComponent}
 PelvisLocation {type=Vector}
]
